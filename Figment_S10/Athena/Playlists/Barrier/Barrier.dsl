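; Barrier wall actor for a Fortnite Athena playlist that displays a countdown clock and plays effects when the wall comes down.
; Events set the clock's digits and fade through a dynamic material created on clients.
; Wall descent triggers dust, staged camera shakes and timeline-driven audio parameters.
; A client timer slides the audio along the wall line to the point nearest the local view target, switching update rate by audibility.
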(event Custom|On%20Timer%20Loop
  (bind _returnvalue (Fortnite|GetLocalFortPlayerControllers true))
  (bind _output (Utilities|Array|Get(acopy) _returnvalue))
  (bind _returnvalue_2 (Pawn|GetViewTarget _output))
  (bind _returnvalue_3 (Transformation|GetActorLocation _returnvalue_2))
  (bind (_returnvalue_x _returnvalue_y) (Math|Vector|BreakVector (Math|Vector|FindClosestPointonLine _returnvalue_3 (Variables|Default|GetStartPoint) (- (Variables|Default|GetEndPoint) (Variables|Default|GetStartPoint)))))
  (bind _scene (Variables|Default|GetScene))
  (bind _returnvalue_4 (Transformation|GetWorldLocation _scene))
  (bind _returnvalue_5 (Transformation|GetForwardVector _scene))
  (bind _returnvalue_6 (* _returnvalue_5 1e+05))
  ;[1 1 1] Audio logic to find closest point on line and position audio there
  (Variables|Default|SetStartPoint (+ _returnvalue_4 _returnvalue_6))
  (Variables|Default|SetEndPoint (- _returnvalue_4 _returnvalue_6))
  (Utilities|IsValid _output
    (:Is%20Valid
      (Utilities|IsValid _returnvalue_2
        (:Is%20Valid
          (Variables|Default|SetViewTargetLocation _returnvalue_3)
          (Variables|Default|SetAudioLocation (Math|Vector|MakeVector _returnvalue_x _returnvalue_y (.z (Variables|Default|GetViewTargetLocation))))
          (Transformation|SetWorldLocation (Variables|Default|GetAudio) (Variables|Default|GetAudioLocation) false false)
          (bind _returnvalue_7 (Audio|AreAnyListenersWithinRange (Variables|Default|GetAudioLocation) (Variables|/Script/Engine.SoundBase|GetMaxDistance (Variables|/Script/Engine.AudioComponent|GetSound (Variables|Default|GetAudio)))))
          (if (Math|Boolean|NotEqual(Boolean) _returnvalue_7 (Variables|Default|GetListenerInRange))
            (bind _output_get (Variables|Default|SetListenerInRange _returnvalue_7))
            (Utilities|Time|SetTimerbyEvent (EventDispatchers|CreateEvent :!event Custom|On%20Timer%20Loop) (select _output_get (Variables|Default|GetUpdateFrequencyAudible) (Variables|Default|GetUpdateFrequencyInaudible)) true)))))))

(event EventUpdateWallTime
  (params (Digit_0 :type Float%20%28single-precision%29) (Digit_1 :type Float%20%28single-precision%29) (Digit_2 :type Float%20%28single-precision%29) (Digit_3 :type Float%20%28single-precision%29))
  (bind _wallmid (Variables|Default|GetWallMID))
  (Utilities|FlowControl|Sequence
    (:then_0
      (Rendering|Material|SetScalarParameterValue _wallmid "Digit_4" Digit_0 :!class /Script/Engine.MaterialInstanceDynamic))
    (:then_1
      (Rendering|Material|SetScalarParameterValue _wallmid "Digit_3" Digit_1 :!class /Script/Engine.MaterialInstanceDynamic))
    (:then_2
      (Rendering|Material|SetScalarParameterValue _wallmid "Digit_2" Digit_2 :!class /Script/Engine.MaterialInstanceDynamic))
    (:then_3
      (Rendering|Material|SetScalarParameterValue _wallmid "Digit_1" Digit_3 :!class /Script/Engine.MaterialInstanceDynamic))))

(event EventBeginPlay
  (Networking|SwitchHasAuthority
    (:Remote
      (bind _returnvalue (Rendering|Material|CreateDynamicMaterialInstance (Variables|Default|GetStaticMesh) 0 0 "None" :!class /Script/Engine.PrimitiveComponent))
      (Variables|Default|SetWallMID _returnvalue)
      (Utilities|Time|SetTimerbyEvent (EventDispatchers|CreateEvent :!event Custom|On%20Timer%20Loop) (Variables|Default|GetUpdateFrequencyInaudible) true))))

(event EventShowOrHideTimer
  (params (bNewVisibleState :type Boolean))
  (if bNewVisibleState
    (bind (_direction _clock_fade) (Timeline|FadeClockIn Play)
      (:Update
        (Rendering|Material|SetScalarParameterValue (Variables|Default|GetWallMID) "ClockFade" _clock_fade :!class /Script/Engine.MaterialInstanceDynamic)))
    (else
      (Timeline|FadeClockIn Reverse))))

(event EventWallStartingToComeDown
  (params (bIsOnBurgerSide :type Boolean))
  (CallFunction|BeginMovingAudio 16.5)
  (Components|Activation|Activate (Variables|Default|GetP_Barrier_BarrierFading_GroundDust) false)
  (Camera|PlayWorldCameraShake "/Figment_S10/Athena/Playlists/Barrier/Barrier_ScreenStart.Barrier_ScreenStart_C" "0, 0, 0" 0.0 1.25e+05 0.0)
  (Utilities|FlowControl|Delay 1e+01)
  (Camera|PlayWorldCameraShake "/Figment_S10/Athena/Playlists/Barrier/Barrier_ScreenRumble.Barrier_ScreenRumble_C" "0, 0, 0" 0.0 1.25e+05 0.0)
  (Utilities|FlowControl|Delay 6.5)
  (Camera|PlayWorldCameraShake "/Figment_S10/Athena/Playlists/Barrier/Barrier_ScreenHit.Barrier_ScreenHit_C" "0, 0, 0" 0.0 1.25e+05 0.0))

(event Custom|BeginMovingAudio
  (params (Time :type Float%20%28double-precision%29))
  (Audio|Components|Audio|Play (Variables|Default|GetAudio) 0.0)
  (Components|Timeline|SetPlayRate (Variables|Barrier|GetMovement%20Audio) (/ 1.0 Time))
  (Audio|SpawnSoundAttached (Variables|Default|GetSound_OnStartMoving) (Variables|Default|GetAudio))
  (bind (_direction _movingparam1 _movingparam2 _movingparam3) (Timeline|Movement%20Audio PlayFromStart)
    (:Update
      (Audio|Parameter|SetFloatParameter (Variables|Default|GetAudio) "MovingParam1" _movingparam1 :!class /Script/Engine.AudioComponent)
      (Audio|Parameter|SetFloatParameter (Variables|Default|GetAudio) "MovingParam2" _movingparam2 :!class /Script/Engine.AudioComponent)
      (Audio|Parameter|SetFloatParameter (Variables|Default|GetAudio) "MovingParam2Inv" (- 1.0 _movingparam2) :!class /Script/Engine.AudioComponent)
      (Audio|Parameter|SetFloatParameter (Variables|Default|GetAudio) "MovingParam3" _movingparam3 :!class /Script/Engine.AudioComponent))
    (:Finished
      (Audio|SpawnSoundAttached (Variables|Default|GetSound_OnStopMoving) (Variables|Default|GetAudio))
      (Audio|Components|Audio|FadeOut (Variables|Default|GetAudio) 0.5 0.0 "Linear" :!class /Script/Engine.AudioComponent))))

(fn ConstructionScript)
