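# Translucent unlit material for the Barrier playlist's energy wall, with an optional countdown clock.
# Layers depth-based intersection glow, bump-offset scanlines, opposing panned noise and a pulsing grid whose tiling changes with camera distance, with cheaper mobile branches.
# UseClock blends in parameter-driven digits from a number atlas on UV2, and UV1 drives a top-edge glow.
# Opacity derives from desaturated color, blending toward a radial fade at distance.
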Material /Figment_S10/Athena/Playlists/Barrier/Materials/M_Barrier
graph {
 properties {bEnableResponsiveAA=true, blendMode=BLEND_Translucent, pixelDepthOffsetMode=PDOM_Legacy, shadingModel=MSM_Unlit, usedShadingModels=Unlit}
 nodes [
  n1 = BumpOffset(
   Coordinate=Multiply(A=TextureCoordinate(), B=ScalarParameter() {defaultValue=1, parameterName=TilingEnergyNoise})
   Height=ScalarParameter() {defaultValue=-10.5922, parameterName=Height}
  )
  n2 = ScalarParameter() {defaultValue=0.0762, parameterName=PanSpeed}
  n3 = TextureSample(UVs=Panner(Coordinate=@n1, Speed=@n2) {bFractionalPart=true}) {samplerType=SAMPLERTYPE_Grayscale, texture=/Figment_S10/Effects/Fort_Effects/Textures/Tile/LowResBlurredNoise}
  n4 = TextureSample(UVs=Panner(Coordinate=@n1, Speed=Multiply(A=@n2) {constB=-1}) {bFractionalPart=true}) {samplerType=SAMPLERTYPE_Grayscale, texture=&LowResBlurredNoise}
  Add(A=@n3.R, B=@n4.R)
  Constant() {r=0.1}
  Constant() {r=250}
  Constant() {r=0.261}
  n5 = DotProduct(A=VertexNormalWS(), B=CameraVectorWS())
  n6 = Divide(A=Subtract(A=SceneDepth().R, B=PixelDepth().R), B=ScalarParameter() {defaultValue=1783.0293, parameterName=Distance Divisor})
  n7 = OneMinus(None=Clamp(None=Multiply(A=@n5, B=@n6)))
  Cosine(None=@n7)
  n8 = LinearInterpolate(
   A=ScalarParameter() {defaultValue=0.5185, parameterName=TopEdgeGlow_Contrast_1}
   B=ScalarParameter() {defaultValue=-168.1556, parameterName=TopEdgeGlow_Contrast_2}
   Alpha=ComponentMask(None=TextureCoordinate() {coordinateIndex=1}) {g=true}
  )
  n9 = Sine(None=Divide(A=@n7))
  n10 = Ceil(None=@n7)
  n11 = Power(Base=Multiply(A=OneMinus(None=@n9), B=@n10), Exp=ScalarParameter() {defaultValue=1.522, parameterName=Intersection Edge Exponent})
  n12 = VectorParameter() {parameterName=Intersection Edge Color, defaultValue=<rgba 1 0.0422 0 1>}
  n13 = Panner(
   Coordinate=Multiply(A=TextureCoordinate(), B=ScalarParameter() {defaultValue=212.6826, parameterName=ScanlineTiling})
   Speed=ScalarParameter() {defaultValue=-0.5, parameterName=ScalineSpeed}
  ) {bFractionalPart=true}
  n14 = Divide(A=ScalarParameter() {defaultValue=-2027.0364, parameterName=BumpOffsetBackWall}, B=DotProduct(A=VertexNormalWS(), B=CameraVectorWS()))
  n15 = Multiply(
   A=OneMinus(None=@n10)
   B=TextureSample(UVs=BumpOffset(Coordinate=@n13, Height=@n14)) {samplerType=SAMPLERTYPE_LinearColor, texture=/Game/Athena/SafeZone/Lines}.R
  )
  n16 = Add(A=Distance(A=WorldPosition().XYZ, B=CameraPositionWS()), B=Constant() {r=-500})
  n17 = ScalarParameter() {defaultValue=1100.8998, parameterName=Distance_2}
  n18 = OneMinus(None=Saturate(None=Divide(A=@n16, B=@n17)))
  n19 = Add(
   A=Add(A=TextureSample(UVs=@n13) {samplerType=SAMPLERTYPE_LinearColor, texture=&Lines}.R, B=Multiply(A=@n15, B=@n18))
   B=ScalarParameter() {defaultValue=0.3095, parameterName=MinOpacity}
  )
  n20 = VectorParameter() {parameterName=scanlines Color, defaultValue=<rgba 0.8 0.0991 0.0113 1>}
  n21 = TextureCoordinate()
  n22 = ScalarParameter() {defaultValue=10, parameterName=BlocksTiling_2}
  n23 = Multiply(A=TextureCoordinate(), B=ScalarParameter() {defaultValue=1, parameterName=BlocksTiling})
  n24 = TextureSample(UVs=@n23) {samplerType=SAMPLERTYPE_Masks, texture=/Game/Athena/Playlists/Barrier/Textures/T_BarrierGrid_M}
  n25 = ScalarParameter() {defaultValue=25728.8906, parameterName=Distance_3}
  n26 = LinearInterpolate(
   A=TextureSample(UVs=Multiply(A=@n21, B=@n22)) {samplerType=SAMPLERTYPE_Masks, texture=&T_BarrierGrid_M}.RGB
   B=@n24.RGB
   Alpha=Saturate(None=Divide(A=@n16, B=@n25))
  )
  n27 = FeatureLevelSwitch(Default=@n26, ES2=@n24.RGB, ES3_1=@n24.RGB, SM4_REMOVED=@n26)
  n28 = ComponentMask(None=@n27) {r=true}
  n29 = Ceil(None=Abs(None=@n23))
  n30 = ComponentMask(None=@n29) {r=true}
  n31 = ComponentMask(None=@n29) {g=true}
  n32 = Multiply(A=Add(A=Time() {period=1}, B=Add(A=@n30, B=@n31)), B=ScalarParameter() {defaultValue=0.1467, parameterName=Sine Speed Grid})
  n33 = Sine(None=Frac(None=Add(A=@n28, B=@n32)))
  n34 = ComponentMask(None=@n27) {g=true}
  n35 = Multiply(A=Multiply(A=@n3.R, B=@n4.R), B=ScalarParameter() {defaultValue=0.3429, parameterName=NoiseBrightness})
  n36 = Multiply(A=@n4.R, B=ScalarParameter() {defaultValue=0.1905, parameterName=NoiseBrightness_Mobile})
  n37 = Add(
   A=Power(Base=Multiply(A=@n33, B=@n34), Exp=ScalarParameter() {defaultValue=5, parameterName=Grid Power})
   B=FeatureLevelSwitch(Default=@n35, ES2=@n36, ES3_1=@n36, SM4_REMOVED=@n35)
  )
  n38 = TextureSample(UVs=Divide(A=TextureCoordinate(), B=ScalarParameter() {defaultValue=1, parameterName=BlocksTiling_1})) {samplerType=SAMPLERTYPE_Masks, texture=&T_BarrierGrid_M}
  n39 = Add(
   A=Add(A=Multiply(A=@n11, B=@n12.RGB), B=Multiply(A=@n19, B=@n20.RGB))
   B=LinearInterpolate(
    A=VectorParameter() {parameterName=NoiseGlow, defaultValue=<rgba 0.2969 0.0732 0.0062 1>}.RGB
    B=VectorParameter() {parameterName=NoiseGlow_1, defaultValue=<rgba 5 0.7935 0.5208 1>}.RGB
    Alpha=Multiply(A=@n37, B=@n38.B)
   )
  )
  n40 = MaterialFunctionCall(
   UVs=TextureCoordinate() {coordinateIndex=2}
   Texture Scale=AppendVector(A=ScalarParameter() {defaultValue=0.9128, parameterName=u}, B=ScalarParameter() {defaultValue=0.3143, parameterName=v})
  ) {materialFunction=&ScaleUVsByCenter}
  n41 = ScalarParameter() {defaultValue=0.0571, parameterName=MinCountdownField}
  n42 = Multiply(A=ComponentMask(None=@n40.UVs) {r=true}, B=Constant() {desc=Number of digits, r=5}) {constB=5}
  n43 = Constant() {r=4}
  n44 = ComponentMask(None=@n42) {r=true}
  n45 = Constant() {r=1}
  n46 = ScalarParameter() {defaultValue=4, parameterName=Digit_4}
  n47 = ScalarParameter() {defaultValue=3, parameterName=Digit_3}
  n48 = If(
   A=Constant() {r=3}
   B=@n44
   "A > B"=If(A=Constant() {r=2}, B=@n44, "A > B"=If(A=@n45, B=@n44, "A > B"=@n46, "A < B"=@n47), "A < B"=Constant() {r=-1})
   "A < B"=ScalarParameter() {defaultValue=2, parameterName=Digit_2}
  )
  n49 = ScalarParameter() {defaultValue=1, parameterName=Digit_1}
  n50 = Add(A=Frac(None=@n42), B=Floor(None=If(A=@n43, B=@n44, "A > B"=@n48, "A < B"=@n49)))
  n51 = TextureCoordinate()
  n52 = TextureSample!(
   UVs=AppendVector(A=Divide(A=@n50) {constB=11}, B=ComponentMask(None=@n40.UVs) {g=true})
   Tex
   "DDX(UVs)"=DDX(Value=@n51)
   "DDY(UVs)"=DDY(Value=@n51)
   Apply View MipBias
  ) {mipValueMode=TMVM_Derivative, samplerType=SAMPLERTYPE_Masks, texture=/Game/Athena/Playlists/Barrier/Textures/T_Numbers2_M}
  n53 = Multiply(
   A=VectorParameter() {parameterName=CountDownColor, defaultValue=<rgba 15 1.702 0.0571 1>}.RGB
   B=Multiply(A=@n40.0-1 mask, B=Add(A=@n41, B=@n52.RGB))
  )
  n54 = Clamp(
   None=Multiply(A=@n40.0-1 mask, B=ScalarParameter() {parameterName=ClockFade})
   Max=ScalarParameter() {defaultValue=0.9026, parameterName=ClockBlend}
  )
  n55 = Add(
   A=Multiply(A=Clamp(None=@n8), B=VectorParameter() {parameterName=TopEdgeGlow, defaultValue=<rgba 1 0.7121 0.6198 1>}.RGB)
   B=StaticSwitchParameter(True=LinearInterpolate(A=@n39, B=@n53, Alpha=@n54), False=@n39) {defaultValue=true, parameterName=UseClock}
  )
  Desaturation(None=@n55)
  n56 = Multiply(A=WorldPosition().XYZ, B=VectorParameter() {parameterName=RadialScale, defaultValue=<rgba 1 1 -2.9634 1>}.RGB)
  n57 = Add(A=ObjectPositionWS(), B=VectorParameter() {parameterName=CenterOffset, defaultValue=<rgba 0 0 -27121.1953 1>}.RGB)
  n58 = Divide(
   A=Add(A=Distance(A=@n56, B=@n57), B=ScalarParameter() {defaultValue=-84873.2109, parameterName=RadialOffset})
   B=ScalarParameter() {defaultValue=95819.1016, parameterName=RadialFade}
  )
  n59 = ScalarParameter() {defaultValue=10144.4004, parameterName=Distance}
  LinearInterpolate(
   A=Clamp(None=Desaturation(None=@n55)) {desc=opacity}
   B=LinearInterpolate(A=ScalarParameter() {defaultValue=0.9905, parameterName=OpacityFar}, B=Constant(), Alpha=Clamp(None=@n58))
   Alpha=Saturate(None=Divide(A=@n16, B=@n59))
  )
  MaterialFunctionCall!(Distance Bias=Constant(), Distance Divisor=ScalarParameter() {defaultValue=2000, parameterName=Wall ThickNess}, Scene Depth, Pixel Depth) {materialFunction=/Game/Effects/Fort_Effects/Materials/_Functions/MF_DistanceBehindTranslucency_01}
  MaterialFunctionCall!(
   TextureObject=TextureObject() {samplerType=SAMPLERTYPE_LinearColor, texture=/Game/Gadgets/Assets/Teleporter/Textures/T_Teleporter_GridSquares_01}
   TextureSize=ScalarParameter() {defaultValue=1477.552, parameterName=Blocks_TilingSize}
   WorldPosition
   Export Float 4=StaticBool()
   In
   World Space Normal
   ProjectionTransitionContrast=Constant() {r=13}
  ) {materialFunction=/Game/Athena/SafeZone/WorldAlignedTexture_ScalePan}
  PixelDepth()
  ScalarParameter() {defaultValue=512, parameterName=GridWorldScale}
  ScalarParameter() {defaultValue=13, parameterName=GridProjContrast}
  Time() {bOverride_Period=true, period=3.8314}
 ]
}
