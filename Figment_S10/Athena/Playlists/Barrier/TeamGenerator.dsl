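; Team generator objective for the Barrier playlist: a destructible actor with a swappable food-team head mesh.
; BeginPlay applies damage-resistance effects after a delay and spawns a team-colored beam; a separate event applies a damage-over-time timer effect.
; Damage and health-state events drive wobble and cracks through a head dynamic material, plus effects, audio and a friendly-team klaxon.
; Death calls OnGeneratorDestroyed, hides the actor and plays team-specific destruction effects with camera shake.
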(event EventBeginPlay
  (bind _returnvalue (Transformation|GetActorLocation))
  ;[1 1 1] hide health bar on beginplay
  (Utilities|FlowControl|Delay 1.0)
  (GameplayEffects|ApplyGameplayEffectToSelf :self (Ability|GetAbilitySystemComponent self) :GameplayEffectClass "/Figment_S10/Athena/Playlists/Barrier/GE_DamageResistanceVsHeavySniper.GE_DamageResistanceVsHeavySniper_C" :Level 0.0)
  (GameplayEffects|ApplyGameplayEffectToSelf :self (Ability|GetAbilitySystemComponent self) :GameplayEffectClass "/Figment_S10/Athena/Playlists/Barrier/GE_DamageResistanceVsPickaxe.GE_DamageResistanceVsPickaxe_C" :Level 0.0)
  (GameplayEffects|ApplyGameplayEffectToSelf :self (Ability|GetAbilitySystemComponent self) :GameplayEffectClass "/Figment_S10/Athena/Playlists/Barrier/GE_DamageResistanceVsTurret.GE_DamageResistanceVsTurret_C" :Level 0.0)
  (if (CallFunction|IsFriendlyTeam)
    (Effects|SpawnEmitteratLocation (Variables|Default|GetBlue%20Beam) _returnvalue)
    (else
      (Effects|SpawnEmitteratLocation (Variables|Default|GetRed%20Beam) _returnvalue))))

(event EventOnDeathServer
  (params (Damage :type Float%20%28single-precision%29) (DamageTags :type Gameplay%20Tag%20Container%20Structure%20%28by%20ref%29) (Momentum :type Vector :default "0, 0, 0") (HitInfo :type Hit%20Result%20Structure%20%28by%20ref%29) (InstigatedBy :type Controller%20Object%20Reference) (DamageCauser :type Actor%20Object%20Reference) (EffectContext :type Gameplay%20Effect%20Context%20Handle%20Structure))
  ;[1 1 1] OnDeathServer
  (CallFunction|OnGeneratorDestroyed))

(event EventOnDeathPlayEffects
  (params (Damage :type Float%20%28single-precision%29) (DamageTags :type Gameplay%20Tag%20Container%20Structure%20%28by%20ref%29) (Momentum :type Vector :default "0, 0, 0") (HitInfo :type Hit%20Result%20Structure%20%28by%20ref%29) (InstigatedBy :type Fort%20Pawn%20Object%20Reference) (DamageCauser :type Actor%20Object%20Reference) (EffectContext :type Gameplay%20Effect%20Context%20Handle%20Structure))
  (bind _returnvalue (Transformation|GetActorLocation))
  ;[1 1 1] OnDeathPlayeffects
  (Rendering|SetActorHiddenInGame true)
  (switch Utilities|FlowControl|Switch|SwitchonEBarrierFoodTeam (CallFunction|GetFoodTeam)
    (:Burger
      (Effects|SpawnEmitteratLocation (Variables|Default|GetDestruction%20FX%20Burger) _returnvalue))
    (:Tomato
      (bind _returnvalue_2 (Effects|SpawnEmitteratLocation (Variables|Default|GetDestruction%20FX%20Tomato) _returnvalue))))
  (Effects|SpawnEmitteratLocation (Variables|Default|GetDestruction%20FX%20Base) _returnvalue)
  (Components|Activation|SetActive _returnvalue_2 true false)
  (Audio|PlaySoundatLocation (Variables|Default|GetDestruction%20Audio) _returnvalue)
  (Camera|PlayWorldCameraShake "/Figment_S10/Athena/Playlists/Barrier/Barrier_ScreenRumble_Short.Barrier_ScreenRumble_Short_C" "0, 0, 0" 0.0 1.25e+05 0.0)
  (Components|DestroyComponent (Variables|Default|GetP_Barrier_Base_BeamIndicator_Yellow)))

(event EventApplyTimerGE
  (params (TimerLength :type Float%20%28single-precision%29))
  (GameplayEffects|ApplyGameplayEffectToSelf :self (Ability|GetAbilitySystemComponent self) :GameplayEffectClass "/Figment_S10/Athena/Playlists/Barrier/GE_Barrier_GeneratorDot.GE_Barrier_GeneratorDot_C" :Level 0.0))

(event EventOnDamagePlayEffects
  (params (Damage :type Float%20%28single-precision%29) (DamageTags :type Gameplay%20Tag%20Container%20Structure%20%28by%20ref%29) (Momentum :type Vector :default "0, 0, 0") (HitInfo :type Hit%20Result%20Structure%20%28by%20ref%29) (InstigatedBy :type Fort%20Pawn%20Object%20Reference) (DamageCauser :type Actor%20Object%20Reference) (EffectContext :type Gameplay%20Effect%20Context%20Handle%20Structure))
  (bind (_bblockinghit _binitialoverlap _time _distance _location _impactpoint _normal) (Collision|BreakHitResult HitInfo))
  (bind (_asb_melee_impact_pickaxe_athena _bsuccess) (Utilities|Casting|CastToB_Melee_Impact_Pickaxe_Athena DamageCauser))
  (bind _head_mid (Variables|Default|GetHead%20MID))
  ;[0.000304 0.159 0.3613] Play extra hit impact fx when hit with pickaxe or bullets
  (Effects|SpawnEmitteratLocation (select _bsuccess "/Figment_S10/Athena/Effects/Particles/Barrier/P_Barrier_Base_Impact_Damage_Effect.P_Barrier_Base_Impact_Damage_Effect" "/Figment_S10/Athena/Effects/Particles/Barrier/P_Barrier_Base_Impact_Damage_Effect.P_Barrier_Base_Impact_Damage_Effect") _location (Math|Conversions|RotationFromXVector _normal))
  ;[0.000304 0.159 0.3613] Setup MID for wobble and cracks material
  (bind (_direction _hit) (Timeline|Timeline_0 PlayFromStart)
    (:Update
      (Rendering|Material|SetScalarParameterValue _head_mid "Code_HitAnimation" _hit :!class /Script/Engine.MaterialInstanceDynamic)
      (Rendering|Material|SetScalarParameterValue _head_mid "WSDamageRadius" 1e+02 :!class /Script/Engine.MaterialInstanceDynamic)
      (Rendering|Material|SetVectorParameterValue (Variables|Default|GetHead%20MID) "Code_DeformationVector" "(R=0.000000,G=0.000000,B=1.000000,A=1.000000)" :!class /Script/Engine.MaterialInstanceDynamic)
      (Rendering|Material|SetVectorParameterValue (Variables|Default|GetHead%20MID) "WSDeformationCenter" (Math|Conversions|ToLinearColor(Vector) (Transformation|GetActorLocation)) :!class /Script/Engine.MaterialInstanceDynamic))))

(event EventOnNewDamageState
  (params (NewState :type EBarrierObjectiveDamageState%20Enum :default "Health_75"))
  ;[1 1 1] Damage State Event
  (switch Utilities|FlowControl|Switch|SwitchonEBarrierObjectiveDamageState NewState
    (:Health_75
      _)
    (:Health_50
      _)
    (:Health_25
      _)
    (:Health_10
      _)
    (:Health_5
      ;[1 1 1] Klaxxon warnings at 5/4/3/2/1 damage #'s
      (if (CallFunction|IsFriendlyTeam)
        (Audio|SpawnSoundAttached (Variables|Default|GetKlaxxonWarningAudio) (Variables|Default|GetHeadMesh))))
    (:Health_4
      _)
    (:Health_3
      _)
    (:Health_2
      _)
    (:Health_1
      _))
  ;[0.000304 0.159 0.3613] Spawn damagestate fx
  (Effects|SpawnEmitteratLocation "/Figment_S10/Athena/Effects/Particles/Barrier/P_BarrierBase_DamageState.P_BarrierBase_DamageState" (Transformation|GetActorLocation))
  ;[1 1 1] Sets the amount of cracks
  (Rendering|Material|SetScalarParameterValue (Variables|Default|GetHead%20MID) "Code_DestructionAmount" (Health|GetHealthPercent) :!class /Script/Engine.MaterialInstanceDynamic)
  ;[1 1 1] Damage audio
  (Audio|PlaySoundatLocation (Variables|Default|GetBigDamageStateAudio) (Transformation|GetActorLocation)))

(event EventOnHeadRotationSet
  (params (NewYaw :type Float%20%28single-precision%29))
  (Transformation|SetWorldRotation (Variables|Default|GetHeadMesh) (Math|Rotator|MakeRotator 0.0 0.0 (+ NewYaw 9e+01)) false false))

(event EventSetHeadMesh
  (params (NewMesh :type Static%20Mesh%20Object%20Reference) (NewScale :type Vector :default "0, 0, 0") (NewOffset :type Vector :default "0, 0, 0") (MaterialsToSwap :type Array%20of%20Material%20Interface%20Object%20References))
  ;[1 1 1] Switches heads at runtime, set MID for wobble and damage cracks
  (Components|StaticMesh|SetStaticMesh (Variables|Default|GetHeadMesh) NewMesh)
  (for (_array_element _array_index) MaterialsToSwap
    (Rendering|Material|SetMaterial (Variables|Default|GetHeadMesh) _array_index _array_element))
  (bind _returnvalue (Rendering|Material|CreateDynamicMaterialInstance (Variables|Default|GetHeadMesh) 0 0 "None" :!class /Script/Engine.PrimitiveComponent))
  (Variables|Default|SetHead%20MID _returnvalue))

(fn ConstructionScript
  (bind _returnvalue (Rendering|Material|CreateDynamicMaterialInstance (Variables|Default|GetHeadMesh) 0 0 "None" :!class /Script/Engine.PrimitiveComponent))
  (Variables|Default|SetHead%20MID _returnvalue))
