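; Blueprint actor for a food-themed team objective in the Barrier playlist, wrapping the objective actor as a child actor.
; GetObjectiveActor returns the child actor component's spawned actor cast to AthenaBarrierObjective.
; BeginPlay and the construction script contain no logic.
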(event EventBeginPlay)

(fn GetObjectiveActor
  (bind _asathena_barrier_objective (Utilities|Casting|CastToAthenaBarrierObjective (Variables|/Script/Engine.ChildActorComponent|GetChildActor (Variables|Default|GetChildActor))))
  (return _asathena_barrier_objective))

(fn ConstructionScript)
